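# Music stinger cue for the Storm King weak point reveal, routed to the boss stinger music sound class.
# A fixed short delay precedes a random pick between two equally weighted stinger mixes.
# Playback is non-spatialized at slightly reduced volume.
# A concurrency override allows one instance at a time and blocks new triggers while it plays.
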SoundCue /Game/Sounds/Fort_Music/Orchestral_Score/Orch_Storm_King/StormKing_WeakPoint_Stinger_Cue
root=Delay_0
nodes [
 {id=Delay_0, children [Random_1], properties {delayMax=0.75, delayMin=0.75}}
 {id=Random_1, children [WavePlayer_3,WavePlayer_2], properties {weights [1,1]}}
 {
  id=WavePlayer_2
  properties {soundWaveAssetPtr=/Game/Sounds/Fort_Music/Orchestral_Score/Orch_Storm_King/Final_MIxes/StormKing_Weak_Point_Reveal_Stinger_01}
 }
 {
  id=WavePlayer_3
  properties {soundWaveAssetPtr=/Game/Sounds/Fort_Music/Orchestral_Score/Orch_Storm_King/Final_MIxes/StormKing_Weak_Point_Reveal_Stinger_02}
 }
]
properties {
 bOverrideConcurrency=true
 maxDistance=2097152
 soundClassObject=/Game/Sounds/FortSoundClassesAndModes/Fort_Music_Boss_Stinger
 volumeMultiplier=0.8
 attenuationOverrides {bSpatialize=false}
 concurrencyOverrides {maxCount=1, resolutionRule=PreventNew}
}
